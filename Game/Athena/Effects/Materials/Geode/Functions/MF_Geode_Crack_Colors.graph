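# Material function for the Geode effect that outputs Crack Color and Cloud Color, each blended from normal to corrupted tint parameters.
# The Corruption scalar shapes a clamped power gradient on UV green, widening and strengthening it as it rises.
# A static switch can replace that computed mask with the Alternate UVs input.
# Crack color also passes through Color_Time_of_Day, dimmed by day and tinted at night; cloud color does not.
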MaterialFunction /Game/Athena/Effects/Materials/Geode/Functions/MF_Geode_Crack_Colors
graph {nodes [
  n1 = ComponentMask(None=TextureCoordinate()) {g=true}
  n2 = ScalarParameter() {parameterName=Corruption}
  n3 = LinearInterpolate(Alpha=@n2) {constA=10, constB=4}
  n4 = Clamp(None=Multiply(A=Power(Base=@n1, Exp=@n3), B=Multiply(A=@n2) {constB=5}) {constB=2})
  n5 = StaticSwitch(
   True=FunctionInput(Preview=@n4) {inputName=Alternate UVs, inputType=FunctionInput_Vector2}
   False=@n4
   Value=FunctionInput(Preview=StaticBool()) {bUsePreviewValueAsDefault=true, inputName=Use Alternate UVs, inputType=FunctionInput_StaticBool}
  )
  n6 = LinearInterpolate(
   A=VectorParameter() {parameterName=CrackColor, defaultValue=<rgba 0.4 1.85 2 0>}.RGB
   B=VectorParameter() {parameterName=CorruptCrackColor, defaultValue=<rgba 0.7344 0.0365 0 1>}.RGB
   Alpha=@n5
  )
  n7 = VectorParameter() {parameterName=NightCrackColorMult, defaultValue=<rgba 0.7 0.4124 0.6978 0>}
  FunctionOutput(None=MaterialFunctionCall(Color_Morning=@n6, Color_Day=Multiply(A=@n6) {constB=0.7}, Color_Evening=@n6, Color_Night=Multiply(A=@n6, B=@n7.RGB)) {materialFunction=&Color_Time_of_Day}.Result) {outputName=Crack Color}
  FunctionOutput(None=LinearInterpolate(
    A=VectorParameter() {parameterName=CloudColor, defaultValue=<rgba 0.1761 0.16 0.4792 0>}.RGB
    B=VectorParameter() {parameterName=CorruptCloudColor, defaultValue=<rgba 0.3073 0.0156 0 1>}.RGB
    Alpha=@n5
   )) {outputName=Cloud Color}
  MaterialFunctionCall() {materialFunction=&DebugTimeSine}
 ]}
