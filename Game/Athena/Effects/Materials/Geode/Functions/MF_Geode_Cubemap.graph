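# Material function for the Geode effects that produces a tinted, Fresnel-weighted fake reflection with a sun and moon highlight.
# An offsettable world-space reflection vector samples a skylight cubemap, scaled by the ambient cubemap tint and a color multiplier.
# The same vector, dotted with the sun and moon direction from the Fortnite parameter collection, forms a powered two-color highlight that is zeroed on Low quality.
# Fresnel scales the sum of both terms.
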MaterialFunction /Game/Athena/Effects/Materials/Geode/Functions/MF_Geode_Cubemap
graph {nodes [
  n1 = Add(
   A=ReflectionVectorWS() {bNormalizeCustomWorldNormal=true}
   B=FunctionInput(Preview=Constant()) {bUsePreviewValueAsDefault=true, inputName=UV add}
  )
  n2 = MaterialFunctionCall() {materialFunction=/Engine/Functions/GetPostProcessFunctions/GetAmbientCubemapTint}
  n3 = Multiply(
   A=Multiply(
    A=TextureSampleParameterCube!(UVs=@n1, Apply View MipBias) {parameterName=Cubemap, samplerType=SAMPLERTYPE_LinearColor, texture=/Game/Packages/Fortress_Sky/TexturesHDR/T_AthenaSkylight}.RGB
    B=ComponentMask(None=@n2.AmbientCubemapTint) {b=true, g=true, r=true}
   )
   B=VectorParameter() {parameterName=Cube Map Multiplier, defaultValue=<rgba 0.4413 0.6162 0.8 0>}.RGB
  ) {constB=1.5}
  n4 = CollectionParameter() {
   collection=&FortniteMaterialParameters
   parameterName=SunAndMoonModelDirectionalVector
  }
  n5 = ComponentMask(None=@n4) {b=true, g=true, r=true}
  n6 = Clamp(None=Power(Base=DotProduct(A=@n5, B=@n1), Exp=ScalarParameter() {defaultValue=80, parameterName=Spec Highlight Power}) {constExponent=20})
  n7 = Constant3Vector() {constant=<rgba 1.4793 0.6794 2 0>}
  n8 = Constant3Vector() {constant=<rgba 13.5417 21.5971 25 0>}
  n9 = QualitySwitch(Default=Multiply(A=@n6, B=LinearInterpolate(A=@n7, B=@n8, Alpha=@n6)), Low=Constant())
  FunctionOutput(None=Multiply(A=Add(A=@n3, B=@n9) {constB=0}, B=Fresnel() {baseReflectFraction=0.03, exponent=4}))
 ]}
